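# Translucent unlit Niagara sprite material that draws a wobbling bubble, named as a shadow for a simple chrome bubble effect.
# A panning swirl normal, offset per particle, distorts the bubble texture UVs, scaled by WiggleIntensity from a dynamic parameter.
# The bubble's green channel, particle alpha and a vertex-interpolated near-camera fade multiply together, while particle color feeds an emissive adjustment function.
# The NearFade and WiggleIntensity scalar parameters are unconnected; dynamic parameters supply both.
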Material /Game/NeverCook/FX/LackLuster/Materials/M_Chrome_Simple_Bubble_Shadow
graph {
 properties {bUsedWithNiagaraSprites=true, blendMode=BLEND_Translucent, pixelDepthOffsetMode=PDOM_Legacy, shadingModel=MSM_Unlit, usedShadingModels=Unlit}
 nodes [
  MaterialFunctionCall(Emissive=ParticleColor().RGB) {materialFunction=&MF_FX_EmissiveAdjustments}
  n1 = Add(A=TextureCoordinate() {uTiling=0.4, vTiling=0.4}, B=ParticleRandom())
  n2 = ComponentMask(None=TextureSample(UVs=Panner(Coordinate=@n1) {speedY=0.4}) {samplerType=SAMPLERTYPE_Normal, texture=/Game/Global/Textures/Default/Noise/T_SwirlNoise_N}.RGB) {g=true, r=true}
  n3 = DynamicParameter() {parameterIndex=1, defaultValue=<rgba 1 1 0.2 0>, paramNames [Emissive, Erode, WiggleIntensity, Param4]}
  n4 = Clamp(None=Add(A=TextureCoordinate(), B=Multiply(A=@n2, B=@n3.WiggleIntensity)))
  n5 = MaterialFunctionCall(Distance=DynamicParameter() {parameterIndex=3, paramNames [Emissive,Param2,NearFade,Param4]}.NearFade) {materialFunction=&Near_Camera_Fade}
  Multiply(
   A=ParticleColor().A
   B=Multiply(A=TextureSample(UVs=@n4) {texture=/Game/Effects/Fort_Effects/Textures/Water/T_BoilBubble_D}.G, B=VertexInterpolator(VS=@n5.Result).PS)
  )
  ScalarParameter() {defaultValue=1500, parameterName=NearFade}
  ScalarParameter() {defaultValue=0.2, parameterName=WiggleIntensity}
 ]
}
