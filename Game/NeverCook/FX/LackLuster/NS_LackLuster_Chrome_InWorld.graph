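# Niagara system for in-world LackLuster "Chrome" sprite effects, scattering large shapes and smaller blobs over visible scene surfaces.
# GPU emitters spawn in a sphere, snap to distance fields and the depth buffer, and kill invalid or far-moved samples.
# Particles drift upward and fade, driven by curves, camera-distance scratch logic and dynamic material parameters.
# Chrome_Blobs003 spawns secondary blobs from ChromeBigShapes particles; Chrome_Blobs and Chrome_Blobs001 are disabled.
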NiagaraSystem /Game/NeverCook/FX/LackLuster/NS_LackLuster_Chrome_InWorld
dependencies {
 usedDataInterfaces [
  &NiagaraDataInterfaceCamera
  &NiagaraDataInterfaceGBuffer
  &NiagaraDataInterfaceCollisionQuery
  &NiagaraDataInterfaceColorCurve
  &NiagaraDataInterfaceCurve
  &NiagaraDataInterfaceParticleRead
 ]
 usedDynamicInputs []
 usedModules [&SystemState]
}
emitters {
 ChromeBigShapes {
  simTarget=GPUComputeSim
  emitterUpdateScript [
   /SetVariables(Emitter.CameraPostion:Vector3f=&ReturnCameraVector(Camera Property=::NewEnumerator0, Coordinate Space=::Simulation)) {isSetParameters=true, module=SetVariables_#1}
   &EmitterState()
   &SpawnRate(SpawnRate=&Multiply_Float(A=15, B=1), Use Spawn Probability=true)
  ]
  particleSpawnScript [
   &InitializeParticle(
    AdjustAlpha=true
    AdjustHue=true
    AdjustSaturation=true
    AdjustValue=true
    Color=<rgba 0.1406 0.0897 0.9792 1>
    Color Mode=::NewEnumerator3
    Lifetime Max=4
    Lifetime Min=3
    Lifetime Mode=::NewEnumerator1
    Saturation Range:Vector2f=<0.25 1>
    Sprite Rotation Angle Max=5
    Sprite Rotation Angle Min=-5
    Sprite Rotation Mode=::NewEnumerator2
    Sprite Size Max=<600 800>
    Sprite Size Min=<400 600>
    Sprite Size Mode=::NewEnumerator2
    UsePositionOffset=false
    Value Range:Vector2f=<0.25 0.5>
    Write Material Random=true
   ) {module=InitializeParticle001}
   /Niagara/Modules/Spawn/Initialization/InitializeParticle(
    Color=&LerpLinearColors(StartColor=<rgba 1 1 1 1>, EndColor=<rgba 1 1 1 1>, LerpFactor=1)
    Lifetime=&Multiply_Float(A=&UniformRangedFloat(Minimum=1, Maximum=2), B=1)
    Module.Write Color=true
    Module.Write Lifetime=true
    Module.Write Mass=true
    Module.Write Material Random=true
    Module.Write Position=false
    Module.Write Scale=false
    Module.Write SpriteRotation=false
    Module.Write SpriteSize=true
    Position=&SimulationPosition()
    Sprite Size=&UniformRangedVector2D(
     Minimum=&Vector2DFromFloat(Value=25)
     Maximum=&Vector2DFromFloat(Value=50)
    )
   ) {enabled=false}
   /SetVariables(Particles.VisibilityTag=1) {enabled=false, isSetParameters=true, module=SetVariables_#2}
   &BoxLocation(
    Box Offset=&MakeVector(
     X=15
     Y=15
     Z=&Multiply_Float(A=&MakeFloatFromVector(Vector=$System.BoxSize, Channel=::NewEnumerator2), B=0)
    )
    Box Origin:NiagaraPosition=/Niagara/DynamicInputs/Helpers/SimulationPosition()
    Box Size=$System.BoxSize
    Coordinate Space:ENiagaraCoordinateSpace=::Simulation
    MaskSpawnGroup=false
    Module.Override Seed=false
    Rotate Box=false
    Surface Only=false
   ) {enabled=false}
   /Game/Effects/Niagara/Modules/Spawn/FN_Spawn_GBuffer_Color() {enabled=false}
   /Niagara/Modules/Update/Forces/CurlNoiseForce(Noise Frequency=150, Pan Noise=true, Pan Noise Field=<0 0 -5>) {enabled=false, module=CurlNoiseForce001}
   &ApplyInitialForces() {enabled=false}
   &CameraOffset(Camera Offset Amount=25)
   &SubUVAnimation(End Frame=7, SubUV Animation Mode:ENiagara_SubUVLookupMode=::NewEnumerator2)
   &SphereLocation(
    MaskSpawnGroup=false
    Override Seed=false
    Sphere Radius=1500
    Surface Only=false
    UseNonUniformScale=false
    UseOffset=false
    UseSphereOrientation=false
   )
   /Niagara/Modules/Math/MoveToNearestDistanceFieldSurface_GPU(Kill by Distance=5000, bKillByDistance=true)
   /SetVariables(Particles.InitialPos:NiagaraPosition=$Particles.Position) {isSetParameters=true, module=SetVariables_#3}
   /Niagara/Modules/Math/PlaceParticlesOnDepthBuffer_GPU(Max Z Depth=3500, Module.Use Max Z Depth=true)
   &KillParticles(Kill Particles=&SetBoolByFloatComparison(
     A=&VectorLength(Vector=&Subtract_Vector(A=$Particles.InitialPos, B=Linked() {linkedVariable {name=Particles.Position, type=NiagaraPosition}}))
     B=3500
    ))
   &KillParticles(Kill Particles=&Boolean_NotOperation(Bool=$Output.PlaceParticlesOnDepthBuffer_GPU.ValidSample)) {module=KillParticles001}
   &self:ScratchModule(Angle=15, Normal:Vector3f=<0 0 1>, Velocity=0.1) {module=ScratchModule}
   /Game/Effects/Niagara/Modules/Spawn/FN_Spawn_GBuffer_Color_WithinRange(ColorMaxRange=1, ColorMinRange=0.04) {enabled=false}
  ]
  particleUpdateScript [
   &ParticleState()
   &KillParticlesInVolume(
    Box Size:Vector3f=$System.BoxSize
    Invert Volume=true
    Kill Shape=::NewEnumerator1
    Origin Offset:Vector3f=&MakeVector(
     X=0
     Y=0
     Z=&Multiply_Float(A=&MakeFloatFromVector(Vector=$System.BoxSize, Channel=::NewEnumerator2), B=0)
    )
   ) {enabled=false}
   &self:Camera_to_Particle_Distance_Calculation(Camera Postion:Vector3f=$Emitter.CameraPostion, Distance=5000, Offset=250, Particle Position:NiagaraPosition=$Particles.Position) {module=Camera_to_Particle_Distance_Calculation}
   /SetVariables() {enabled=false, isSetParameters=true, module=SetVariables_#4}
   /SetVariables(Particles.GradientColor:Vector3f=&MakeVectorFromLinearColorRGB(Color=&MakeLinearColorFromVectorAndFloat(
      "Vector (RGB)"=&MakeVectorFromLinearColorRGB(Color=&ColorFromCurve(ColorCurve=DataInterface() {propertyValues {
          AlphaCurve {keys [<at -0.0023 1 RCIM_Cubic>,<at 0.9977 1 RCIM_Cubic>]}
          BlueCurve {keys [<at 0.1175 0.3005 RCIM_Cubic>,<at 0.2682 0.2705>,<at 0.4507 0.2122>,<at 0.5951 0.5271>,<at 0.7691 1.5>,<at 0.9071 0.521 RCIM_Cubic>]}
          GreenCurve {keys [<at 0.1175 0.1878 RCIM_Cubic>,<at 0.2682 0.2623>,<at 0.4507 0.2346>,<at 0.5951 0.4287>,<at 0.7691 0.4532>,<at 0.9071 0.4851 RCIM_Cubic>]}
          RedCurve {keys [<at 0.1175 0.3372 RCIM_Cubic>,<at 0.2682 0.7305>,<at 0.4507 0.7758>,<at 0.5951 0.7084>,<at 0.7691 0>,<at 0.9071 0.807 RCIM_Cubic>]}
         }}))
      "Float (Alpha)"=1
     ))) {enabled=false, isSetParameters=true, module=SetVariables_#5}
   /SetVariables(Particles.VisibilityTag=&MakeCustomIntFromBool(Bool=$Transient.FirstFrame, True INT=1, False INT=0)) {enabled=false, isSetParameters=true, module=SetVariables_#6}
   &GravityForce(Gravity=&Multiply_VectorByFloat(
     Vector=<0 0 25>
     Float=&FloatFromCurve(FloatCurve=DataInterface() {propertyValues {Curve {keys [<at 0 0>,<at 0.25 0>,<at 1 1>]}}})
    ))
   &CurlNoiseForce(
    Curl Noise Cone Mask Angle=150
    Curl Noise Cone Mask Falloff Angle=25
    Mask Curl Noise=true
    Noise Frequency:NiagaraFloat=&RandomRangeFloat(Minimum=250, Maximum=500)
    Noise Strength=&FloatFromCurve(
     FloatCurve=DataInterface() {propertyValues {Curve {keys [
         {interpMode=RCIM_Cubic, tangentMode=RCTM_User, time=0, value=0}
         {interpMode=RCIM_Cubic, tangentMode=RCTM_User, time=0.0675, value=1.5837}
         {interpMode=RCIM_Cubic, tangentMode=RCTM_User, time=0.1302, value=0.1555}
         {interpMode=RCIM_Cubic, tangentMode=RCTM_User, time=0.25, value=1}
        ]}}}
     Scale Curve=100
    )
    Pan Noise=true
    Pan Noise Field=<0 1 80>
    Random Seed=25
   ) {enabled=false}
   &PointAttractionForce(
    Attraction Radius=400
    Attraction Strength=-1
    Attractor Position:Vector3f=$Emitter.CameraPosition
    Attractor Position Offset:Vector3f=<0 0 -500>
    Falloff Exponent=1
    Kill Within Radius=false
    Module.Use Falloff=true
    UseAttractorOffset=true
   ) {enabled=false, module=PointAttractionForce001}
   &Drag(
    Drag=&Multiply_Float(
     A=&UniformRangedFloat(Minimum=0, Maximum=0)
     B=&FloatFromCurve(FloatCurve=DataInterface() {propertyValues {Curve {keys [<at 0.1 1>,<at 0.25 0.1>]}}})
    )
    Use Linear Drag=true
    Use Rotational Drag=false
   ) {enabled=false}
   &SolveForcesAndVelocity(Clamp Velocity=true, Limit Acceleration=false, Speed Limit=250)
   &ScaleColor(
    Scale Alpha=&FloatFromCurve(FloatCurve=DataInterface() {propertyValues {Curve {keys [
         {interpMode=RCIM_Cubic, tangentMode=RCTM_User, time=0, value=0}
         <at 0.1 0.4>
         <at 0.3 0.5>
         {interpMode=RCIM_Cubic, tangentMode=RCTM_User, time=0.4, value=0.95}
         <at 0.6 1>
         {tangentMode=RCTM_User, time=1, value=0}
        ]}}})
    Scale RGB=<3 3 3>
    ScaleA=true
    ScaleRGB=true
   )
   &DynamicMaterialParameters(
    Index 0 Param 1=&RandomRangeFloat(Minimum=1, Maximum=1)
    Index 0 Param 2=1
    Index 0 Param 3=&RandomRangeFloat(Minimum=0.1, Maximum=0.2)
    Index 0 Param 4=1
    Index 1 Param 1=&RandomRangeFloat(Minimum=1, Maximum=1)
    Index 1 Param 2=&RandomRangeFloat(Minimum=1, Maximum=1)
    Index 1 Param 3=&RandomRangeFloat(Minimum=0.5, Maximum=0.5)
    Index 2 Param 1=&RandomRangeFloat(Minimum=0.01, Maximum=0.03)
    Index 3 Param 3=&RandomRangeFloat(Minimum=800, Maximum=1200)
    Module.Param0WriteEnabled=true
    Module.Param0WriteEnabled1=true
    Module.Param0WriteEnabled2=true
    Module.Param0WriteEnabled3=true
    Module.Param1WriteEnabled=true
    Module.Param1WriteEnabled1=true
    Module.Param1WriteEnabled2=true
    Module.Param1WriteEnabled3=true
    Module.Param2WriteEnabled=true
    Module.Param2WriteEnabled1=true
    Module.Param2WriteEnabled2=true
    Module.Param2WriteEnabled3=true
    Module.Param3WriteEnabled=true
    Module.Param3WriteEnabled1=true
    Module.Param3WriteEnabled2=true
    Module.Param3WriteEnabled3=true
    Write Parameter Index 1=true
    Write Parameter Index 2=true
    Write Parameter Index 3=true
   )
   /Niagara/Modules/Update/Camera/MaintainInCameraParticleScale() {enabled=false}
   &ScaleSpriteSize(
    Scale Factor=&MakeVector2D(
     X=&FloatFromCurve(
      FloatCurve=DataInterface() {propertyValues {Curve {keys [
          {tangentMode=RCTM_User, time=0, value=0}
          {interpMode=RCIM_Cubic, tangentMode=RCTM_User, time=0.15, value=1}
          <at 0.3 0.5>
          {tangentMode=RCTM_User, time=1, value=0.375}
         ]}}}
      Scale Curve=0.1
     )
     Y=&FloatFromCurve(
      FloatCurve=DataInterface() {propertyValues {Curve {keys [
          {interpMode=RCIM_Cubic, tangentMode=RCTM_User, time=0, value=0}
          {arriveTangent=1.6606, interpMode=RCIM_Cubic, leaveTangent=1.6606, tangentMode=RCTM_User, time=0.4217, value=0.7394}
          {interpMode=RCIM_Cubic, tangentMode=RCTM_User, time=1, value=1}
         ]}}}
      Scale Curve=0.1
     )
    )
    Scale Sprite Size Mode=::NewEnumerator2
   )
   &ScaleSpriteSize(
    Uniform Curve Scale=0.1
    Uniform Curve Sprite Scale=DataInterface() {propertyValues {Curve {keys [
        {interpMode=RCIM_Cubic, tangentMode=RCTM_User, time=0, value=0}
        {interpMode=RCIM_Cubic, tangentMode=RCTM_User, time=0.1, value=1}
        <at 0.35 0.625>
        {interpMode=RCIM_Cubic, tangentMode=RCTM_User, time=0.6, value=0.625}
        {interpMode=RCIM_Cubic, tangentMode=RCTM_User, time=0.8, value=0.875}
        {tangentMode=RCTM_User, time=1, value=0}
       ]}}}
   ) {enabled=false, module=ScaleSpriteSize001}
  ]
  rendererClasses [&NiagaraSpriteRendererProperties]
  renderers [{rendererClass=&NiagaraSpriteRendererProperties}]
 }
 Chrome_Blobs {
  bEnabled=false
  simTarget=GPUComputeSim
  emitterUpdateScript [
   /SetVariables(Emitter.CameraPostion:Vector3f=&ReturnCameraVector(Camera Property=::NewEnumerator0, Coordinate Space=::Simulation)) {isSetParameters=true, module=SetVariables_#1}
   &EmitterState()
   &SpawnRate(SpawnRate=&Multiply_Float(A=500, B=1), Use Spawn Probability=true)
  ]
  particleSpawnScript [
   &InitializeParticle(
    Color=&LerpLinearColors(StartColor=<rgba 1 1 1 1>, EndColor=<rgba 1 1 1 1>, LerpFactor=1)
    Lifetime=&Multiply_Float(A=&UniformRangedFloat(Minimum=1, Maximum=2), B=1)
    Module.Write Color=true
    Module.Write Lifetime=true
    Module.Write Mass=true
    Module.Write Material Random=true
    Module.Write Position=false
    Module.Write Scale=false
    Module.Write SpriteRotation=false
    Module.Write SpriteSize=true
    Position=/Niagara/Functions/Localspace/SimulationPosition()
    Sprite Size=&UniformRangedVector2D(
     Minimum=&Vector2DFromFloat(Value=25)
     Maximum=&Vector2DFromFloat(Value=50)
    )
   ) {enabled=false}
   /SetVariables(Particles.VisibilityTag=1) {enabled=false, isSetParameters=true, module=SetVariables_#2}
   /Niagara/Modules/Spawn/Initialization/V2/InitializeParticle(
    AdjustAlpha=false
    AdjustHue=true
    AdjustSaturation=false
    AdjustValue=false
    Color Mode=::NewEnumerator3
    Lifetime Max=3
    Lifetime Min=2
    Lifetime Mode=::NewEnumerator1
    Sprite Size Max=<15 50>
    Sprite Size Min=<10 25>
    Sprite Size Mode=::NewEnumerator2
    UsePositionOffset=false
    Write Material Random=true
   ) {module=InitializeParticle001}
   &BoxLocation(
    Box Offset=&MakeVector(
     X=15
     Y=15
     Z=&Multiply_Float(A=&MakeFloatFromVector(Vector=$System.BoxSize, Channel=::NewEnumerator2), B=0)
    )
    Box Origin:NiagaraPosition=/Niagara/DynamicInputs/Helpers/SimulationPosition()
    Box Size=$System.BoxSize
    Coordinate Space:ENiagaraCoordinateSpace=::Simulation
    MaskSpawnGroup=false
    Module.Override Seed=false
    Rotate Box=false
    Surface Only=false
   ) {enabled=false}
   &FN_Spawn_GBuffer_Color() {enabled=false}
   &CurlNoiseForce(Noise Frequency=150, Pan Noise=true, Pan Noise Field=<0 0 -5>) {enabled=false, module=CurlNoiseForce001}
   &ApplyInitialForces() {enabled=false}
   &CameraOffset(Camera Offset Amount=50)
   &SphereLocation(
    MaskSpawnGroup=false
    Override Seed=false
    Sphere Radius=1500
    Surface Only=false
    UseNonUniformScale=false
    UseOffset=false
    UseSphereOrientation=false
   )
   &MoveToNearestDistanceFieldSurface_GPU(Kill by Distance=50, bKillByDistance=true)
   /SetVariables(Particles.InitialPos:NiagaraPosition=$Particles.Position) {isSetParameters=true, module=SetVariables_#3}
   &PlaceParticlesOnDepthBuffer_GPU(Max Z Depth=3500, Module.Use Max Z Depth=true)
   &KillParticles(Kill Particles=&SetBoolByFloatComparison(
     A=&VectorLength(Vector=&Subtract_Vector(A=$Particles.InitialPos, B=Linked() {linkedVariable {name=Particles.Position, type=NiagaraPosition}}))
     B=3500
    ))
   &KillParticles(Kill Particles=&Boolean_NotOperation(Bool=$Output.PlaceParticlesOnDepthBuffer_GPU.ValidSample)) {module=KillParticles001}
   &FN_Spawn_GBuffer_Color_WithinRange(ColorMaxRange=1, ColorMinRange=0.04) {enabled=false}
  ]
  particleUpdateScript [
   &ParticleState()
   &KillParticlesInVolume(
    Box Size:Vector3f=$System.BoxSize
    Invert Volume=true
    Kill Shape=::NewEnumerator1
    Origin Offset:Vector3f=&MakeVector(
     X=0
     Y=0
     Z=&Multiply_Float(A=&MakeFloatFromVector(Vector=$System.BoxSize, Channel=::NewEnumerator2), B=0)
    )
   ) {enabled=false}
   &self:Camera_to_Particle_Distance_Calculation(Camera Postion:Vector3f=$Emitter.CameraPostion, Distance=2000, Particle Position:NiagaraPosition=$Particles.Position) {module=Camera_to_Particle_Distance_Calculation}
   /SetVariables() {enabled=false, isSetParameters=true, module=SetVariables_#4}
   /SetVariables(Particles.GradientColor:Vector3f=&MakeVectorFromLinearColorRGB(Color=&MakeLinearColorFromVectorAndFloat(
      "Vector (RGB)"=&MakeVectorFromLinearColorRGB(Color=&ColorFromCurve(ColorCurve=DataInterface() {propertyValues {
          AlphaCurve {keys [<at -0.0023 1 RCIM_Cubic>,<at 0.9977 1 RCIM_Cubic>]}
          BlueCurve {keys [<at 0.1175 0.3005 RCIM_Cubic>,<at 0.2682 0.2705>,<at 0.4507 0.2122>,<at 0.5951 0.5271>,<at 0.7691 1.5>,<at 0.9071 0.521 RCIM_Cubic>]}
          GreenCurve {keys [<at 0.1175 0.1878 RCIM_Cubic>,<at 0.2682 0.2623>,<at 0.4507 0.2346>,<at 0.5951 0.4287>,<at 0.7691 0.4532>,<at 0.9071 0.4851 RCIM_Cubic>]}
          RedCurve {keys [<at 0.1175 0.3372 RCIM_Cubic>,<at 0.2682 0.7305>,<at 0.4507 0.7758>,<at 0.5951 0.7084>,<at 0.7691 0>,<at 0.9071 0.807 RCIM_Cubic>]}
         }}))
      "Float (Alpha)"=1
     ))) {enabled=false, isSetParameters=true, module=SetVariables_#5}
   /SetVariables(Particles.VisibilityTag=&MakeCustomIntFromBool(Bool=$Transient.FirstFrame, True INT=1, False INT=0)) {enabled=false, isSetParameters=true, module=SetVariables_#6}
   &GravityForce(Gravity=&Multiply_VectorByFloat(
     Vector=<0 0 50>
     Float=&FloatFromCurve(FloatCurve=DataInterface() {propertyValues {Curve {keys [<at 0 0>,<at 0.4 0.01>,<at 1 1>]}}})
    ))
   &CurlNoiseForce(
    Curl Noise Cone Mask Angle=150
    Curl Noise Cone Mask Falloff Angle=25
    Mask Curl Noise=true
    Noise Frequency:NiagaraFloat=&RandomRangeFloat(Minimum=250, Maximum=500)
    Noise Strength=&FloatFromCurve(
     FloatCurve=DataInterface() {propertyValues {Curve {keys [
         {interpMode=RCIM_Cubic, tangentMode=RCTM_User, time=0, value=0}
         {interpMode=RCIM_Cubic, tangentMode=RCTM_User, time=0.0675, value=1.5837}
         {interpMode=RCIM_Cubic, tangentMode=RCTM_User, time=0.1302, value=0.1555}
         {interpMode=RCIM_Cubic, tangentMode=RCTM_User, time=0.25, value=1}
        ]}}}
     Scale Curve=100
    )
    Pan Noise=true
    Pan Noise Field=<0 1 80>
    Random Seed=25
   ) {enabled=false}
   &PointAttractionForce(
    Attraction Radius=400
    Attraction Strength=-1
    Attractor Position:Vector3f=$Emitter.CameraPosition
    Attractor Position Offset:Vector3f=<0 0 -500>
    Falloff Exponent=1
    Kill Within Radius=false
    Module.Use Falloff=true
    UseAttractorOffset=true
   ) {enabled=false, module=PointAttractionForce001}
   &Drag(
    Drag=&Multiply_Float(
     A=&UniformRangedFloat(Minimum=0, Maximum=0)
     B=&FloatFromCurve(FloatCurve=DataInterface() {propertyValues {Curve {keys [<at 0.1 1>,<at 0.25 0.1>]}}})
    )
    Use Linear Drag=true
    Use Rotational Drag=false
   ) {enabled=false}
   &SolveForcesAndVelocity(Clamp Velocity=true, Limit Acceleration=false)
   &ScaleColor(
    Scale Alpha=&FloatFromCurve(
     FloatCurve=DataInterface() {propertyValues {Curve {keys [
         {interpMode=RCIM_Cubic, tangentMode=RCTM_User, time=0, value=0}
         {interpMode=RCIM_Cubic, tangentMode=RCTM_User, time=0.5, value=1}
         {tangentMode=RCTM_User, time=1, value=0}
        ]}}}
     Scale Curve=2
    )
    ScaleA=true
    ScaleRGB=true
   )
   &DynamicMaterialParameters(
    Index 0 Param 1=&RandomRangeFloat(Minimum=0, Maximum=1)
    Index 0 Param 2=1
    Index 0 Param 3=0.25
    Index 0 Param 4=15
    Module.Param0WriteEnabled=true
    Module.Param1WriteEnabled=true
    Module.Param2WriteEnabled=true
    Module.Param3WriteEnabled=true
   )
   &ScaleSpriteSize(Scale Factor=&MakeVector2D(
     X=&FloatFromCurve(FloatCurve=DataInterface() {propertyValues {Curve {keys [{tangentMode=RCTM_User, time=0, value=0}, {tangentMode=RCTM_User, time=1, value=1}]}}})
     Y=&FloatFromCurve(FloatCurve=DataInterface() {propertyValues {Curve {keys [{tangentMode=RCTM_User, time=0, value=0}, {tangentMode=RCTM_User, time=0.4, value=1}]}}})
    ))
   &MaintainInCameraParticleScale() {enabled=false}
  ]
  rendererClasses [&NiagaraSpriteRendererProperties]
  renderers [{rendererClass=&NiagaraSpriteRendererProperties}]
 }
 Chrome_Blobs001 {
  bEnabled=false
  simTarget=GPUComputeSim
  emitterUpdateScript [
   /SetVariables(Emitter.CameraPostion:Vector3f=&ReturnCameraVector(Camera Property=::NewEnumerator0, Coordinate Space=::Simulation)) {isSetParameters=true, module=SetVariables_#1}
   &EmitterState()
   &SpawnRate(SpawnRate=&Multiply_Float(A=500, B=1), Use Spawn Probability=true)
  ]
  particleSpawnScript [
   &InitializeParticle(
    AdjustAlpha=true
    AdjustHue=true
    AdjustSaturation=true
    AdjustValue=true
    Alpha Scale Range:Vector2f=<0.8 0.95>
    Color=<rgba 0.6514 0.0816 0.9792 1>
    Color Mode=::NewEnumerator3
    Hue Shift Range:Vector2f=<-0.25 0.25>
    Lifetime Max=5
    Lifetime Min=3
    Lifetime Mode=::NewEnumerator1
    Saturation Range:Vector2f=<0.25 0.5>
    Sprite Rotation Angle Max=0
    Sprite Rotation Mode=::NewEnumerator2
    Sprite Size Max=<125 250>
    Sprite Size Min=<125 250>
    Sprite Size Mode=::NewEnumerator2
    UsePositionOffset=false
    Value Range:Vector2f=<0.25 0.5>
    Write Material Random=true
   ) {module=InitializeParticle001}
   &FN_Spawn_GBuffer_Color() {enabled=false}
   &SphereLocation(
    MaskSpawnGroup=false
    Override Seed=false
    Sphere Radius=2000
    Surface Only=false
    UseNonUniformScale=false
    UseOffset=false
    UseSphereOrientation=false
   )
   &MoveToNearestDistanceFieldSurface_GPU(Kill by Distance=50, bKillByDistance=true)
   /SetVariables(Particles.InitialPos:NiagaraPosition=$Particles.Position) {isSetParameters=true, module=SetVariables_#3}
   &PlaceParticlesOnDepthBuffer_GPU(Max Z Depth=3500, Module.Use Max Z Depth=true)
   &KillParticles(Kill Particles=&SetBoolByFloatComparison(
     A=&VectorLength(Vector=&Subtract_Vector(A=$Particles.InitialPos, B=Linked() {linkedVariable {name=Particles.Position, type=NiagaraPosition}}))
     B=3500
    ))
   &KillParticles(Kill Particles=&Boolean_NotOperation(Bool=$Output.PlaceParticlesOnDepthBuffer_GPU.ValidSample)) {module=KillParticles001}
   &FN_Spawn_GBuffer_Color_WithinRange(ColorMaxRange=1, ColorMinRange=0.04) {enabled=false}
   &SubUVAnimation(End Frame=7, SubUV Animation Mode:ENiagara_SubUVLookupMode=::NewEnumerator2)
   /SetVariables(Particles.SizeStretch:Vector2f=&Multiply_Vector2D(
     A=$Particles.SpriteSize
     B=&MakeVector2D(
      X=&RandomRangeFloat(Minimum=5, Maximum=10)
      Y=&RandomRangeFloat(Minimum=25, Maximum=50)
     )
    )) {isSetParameters=true, module=SetVariables_#7}
   &CameraOffset(Camera Offset Amount=5)
   &self:ScratchModule(Angle=35, Normal:Vector3f=<0 0 1>, Velocity=1) {module=ScratchModule}
  ]
  particleUpdateScript [
   &ParticleState()
   &self:Camera_to_Particle_Distance_Calculation(Camera Postion:Vector3f=$Emitter.CameraPostion, Distance=5000, Offset=250, Particle Position:NiagaraPosition=$Particles.Position) {module=Camera_to_Particle_Distance_Calculation}
   &GravityForce(Gravity=&Multiply_VectorByFloat(
     Vector=<0 0 50>
     Float=&FloatFromCurve(
      FloatCurve=DataInterface() {propertyValues {Curve {keys [<at 0 0>,<at 0.25 0>,<at 1 1>]}}}
      Scale Curve=&MakeFloatFromVector(Vector=$Particles.Initial.Velocity, Channel=::NewEnumerator2)
     )
    ))
   &SolveForcesAndVelocity(Clamp Velocity=true, Limit Acceleration=false, Speed Limit=250)
   &ScaleColor(
    Scale Alpha=&FloatFromCurve(FloatCurve=DataInterface() {propertyValues {Curve {keys [
         {interpMode=RCIM_Cubic, tangentMode=RCTM_User, time=0, value=0}
         {interpMode=RCIM_Cubic, tangentMode=RCTM_User, time=0.1, value=1}
         <at 0.7 1>
         {tangentMode=RCTM_User, time=1, value=0}
        ]}}})
    Scale RGB=&MakeVectorFromLinearColorRGB(Color=$Particles.Initial.Color)
    ScaleA=true
    ScaleRGB=true
   )
   &DynamicMaterialParameters(
    Index 0 Param 1=&RandomRangeFloat(Minimum=-0.2, Maximum=1)
    Index 0 Param 2=1
    Index 0 Param 3=&RandomRangeFloat(Minimum=0, Maximum=0.5)
    Index 0 Param 4=1
    Index 1 Param 1=&RandomRangeFloat(Minimum=0.05, Maximum=0.1)
    Index 1 Param 2=&FloatFromCurve(FloatCurve=DataInterface() {propertyValues {Curve {keys [<at 0 0.9>,<at 1 0.4>]}}}, Scale Curve=0.25)
    Index 1 Param 3=&FloatFromCurve(
     FloatCurve=DataInterface() {propertyValues {Curve {keys [
         <at 0 0>
         {interpMode=RCIM_Cubic, tangentMode=RCTM_User, time=0.05, value=-0.05}
         {interpMode=RCIM_Cubic, tangentMode=RCTM_User, time=0.1, value=1.6}
         {interpMode=RCIM_Cubic, tangentMode=RCTM_User, time=0.2, value=0.1}
         {interpMode=RCIM_Cubic, tangentMode=RCTM_User, time=0.5, value=1}
         {interpMode=RCIM_Cubic, tangentMode=RCTM_User, time=0.6, value=0.7}
         {interpMode=RCIM_Cubic, tangentMode=RCTM_User, time=1, value=3}
        ]}}}
     Scale Curve=&RandomRangeFloat(Minimum=0.1, Maximum=0.1)
    )
    Module.Param0WriteEnabled=true
    Module.Param0WriteEnabled1=true
    Module.Param1WriteEnabled=true
    Module.Param1WriteEnabled1=true
    Module.Param2WriteEnabled=true
    Module.Param2WriteEnabled1=true
    Module.Param3WriteEnabled=true
    Module.Param3WriteEnabled1=true
    Write Parameter Index 1=true
   )
   &MaintainInCameraParticleScale() {enabled=false}
   &ScaleSpriteSize(
    Uniform Curve Scale=0.25
    Uniform Curve Sprite Scale=DataInterface() {propertyValues {Curve {keys [
        {interpMode=RCIM_Cubic, tangentMode=RCTM_User, time=0, value=0}
        {interpMode=RCIM_Cubic, tangentMode=RCTM_User, time=0.1, value=1}
        <at 0.35 0.625>
        {interpMode=RCIM_Cubic, tangentMode=RCTM_User, time=0.6, value=0.625}
        {interpMode=RCIM_Cubic, tangentMode=RCTM_User, time=0.8, value=0.875}
        {tangentMode=RCTM_User, time=1, value=0}
       ]}}}
   )
   &ScaleSpriteSize() {enabled=false, module=ScaleSpriteSize001}
   /SetVariables(Particles.New Vector 2f:Vector2f=&Multiply_Vector2D_ByFloat(
     Float=&FloatFromCurve(FloatCurve=DataInterface() {propertyValues {Curve {keys [<at 0 0>,<at 0.2 1>,<at 0.5 0>]}}}, Scale Curve=10)
     Vector2D=<1 1>
    )) {isSetParameters=true, module=SetVariables_#8}
   /SetVariables(Particles.New Linear Color:LinearColor=&MultiplyLinearColor(
     Linear Color=<rgba 0.0729 0.0185 0.0629 5>
     Scale Factor=&MakeVector4_FromFloat(Float=&FloatFromCurve(FloatCurve=DataInterface() {propertyValues {Curve {keys [<at 0 1>,<at 0.4 0>]}}}))
    )) {isSetParameters=true, module=SetVariables_#9}
   &SpriteFacingAndAlignment(Sprite Facing=<0 0 1>, WriteAlignment=false, WriteFacing=true)
  ]
  rendererClasses [&NiagaraSpriteRendererProperties]
  renderers [
   {rendererClass=&NiagaraSpriteRendererProperties}
   {rendererClass=&NiagaraSpriteRendererProperties, rendererIndex=1}
   {rendererClass=&NiagaraSpriteRendererProperties, rendererIndex=2}
  ]
 }
 Chrome_Blobs002 {
  simTarget=GPUComputeSim
  emitterUpdateScript [
   /SetVariables(Emitter.CameraPostion:Vector3f=&ReturnCameraVector(Camera Property=::NewEnumerator0, Coordinate Space=::Simulation)) {isSetParameters=true, module=SetVariables_#1}
   &EmitterState()
   &SpawnRate(SpawnRate=&Multiply_Float(A=800, B=1), Use Spawn Probability=true)
  ]
  particleSpawnScript [
   &InitializeParticle(
    AdjustAlpha=true
    AdjustHue=true
    AdjustSaturation=true
    AdjustValue=true
    Alpha Scale Range:Vector2f=<0.8 0.95>
    Color=<rgba 0.0818 0.1045 0.9792 1>
    Color Mode=::NewEnumerator3
    Lifetime Max=5
    Lifetime Min=2.5
    Lifetime Mode=::NewEnumerator1
    Saturation Range:Vector2f=<0.25 0.5>
    Sprite Rotation Angle Max=0
    Sprite Rotation Mode=::NewEnumerator2
    Sprite Size Max=<50 60>
    Sprite Size Min=<40 50>
    Sprite Size Mode=::NewEnumerator2
    UsePositionOffset=false
    Value Range:Vector2f=<0.25 0.5>
    Write Material Random=true
   ) {module=InitializeParticle001}
   &FN_Spawn_GBuffer_Color() {enabled=false}
   &SphereLocation(
    MaskSpawnGroup=false
    Override Seed=false
    Sphere Radius=5000
    Surface Only=false
    UseNonUniformScale=false
    UseOffset=false
    UseSphereOrientation=false
   )
   &MoveToNearestDistanceFieldSurface_GPU(Kill by Distance=5000, bKillByDistance=true)
   /SetVariables(Particles.InitialPos:NiagaraPosition=$Particles.Position) {isSetParameters=true, module=SetVariables_#3}
   &PlaceParticlesOnDepthBuffer_GPU(Max Z Depth=3500, Module.Use Max Z Depth=true)
   &KillParticles(Kill Particles=&SetBoolByFloatComparison(
     A=&VectorLength(Vector=&Subtract_Vector(A=$Particles.InitialPos, B=Linked() {linkedVariable {name=Particles.Position, type=NiagaraPosition}}))
     B=3500
    ))
   &KillParticles(Kill Particles=&Boolean_NotOperation(Bool=$Output.PlaceParticlesOnDepthBuffer_GPU.ValidSample)) {module=KillParticles001}
   &FN_Spawn_GBuffer_Color_WithinRange(ColorMaxRange=1, ColorMinRange=0.04) {enabled=false}
   &SubUVAnimation(End Frame=3, SubUV Animation Mode:ENiagara_SubUVLookupMode=::NewEnumerator2)
   /SetVariables(Particles.SizeStretch:Vector2f=&Multiply_Vector2D(
     A=$Particles.SpriteSize
     B=&MakeVector2D(
      X=&RandomRangeFloat(Minimum=5, Maximum=10)
      Y=&RandomRangeFloat(Minimum=25, Maximum=50)
     )
    )) {isSetParameters=true, module=SetVariables_#7}
   &CameraOffset(Camera Offset Amount=-10)
   &self:ScratchModule(Angle=35, Normal:Vector3f=<0 0 1>, Velocity=1) {module=ScratchModule}
  ]
  particleUpdateScript [
   &ParticleState()
   &self:Camera_to_Particle_Distance_Calculation(Camera Postion:Vector3f=$Emitter.CameraPostion, Distance=5000, Offset=250, Particle Position:NiagaraPosition=$Particles.Position) {module=Camera_to_Particle_Distance_Calculation}
   &GravityForce(Gravity=&Multiply_VectorByFloat(
     Vector=&RandomRangeVector(Minimum=<-3 -3 35>, Maximum=<3 3 50>)
     Float=&FloatFromCurve(
      FloatCurve=DataInterface() {propertyValues {Curve {keys [<at 0 0>,<at 0.25 0>,<at 1 1>]}}}
      Scale Curve=&MakeFloatFromVector(Vector=$Particles.Initial.Velocity, Channel=::NewEnumerator2)
     )
    ))
   &SolveForcesAndVelocity(Clamp Velocity=true, Limit Acceleration=false, Speed Limit=250)
   &ScaleColor(
    Scale Alpha=&FloatFromCurve(FloatCurve=DataInterface() {propertyValues {Curve {keys [
         {interpMode=RCIM_Cubic, tangentMode=RCTM_User, time=0, value=0}
         {interpMode=RCIM_Cubic, tangentMode=RCTM_User, time=0.1, value=1}
         <at 0.7 1>
         {tangentMode=RCTM_User, time=1, value=0}
        ]}}})
    Scale RGB=&MakeVectorFromLinearColorRGB(Color=$Particles.Initial.Color)
    ScaleA=true
    ScaleRGB=true
   )
   &DynamicMaterialParameters(
    Index 0 Param 1=&RandomRangeFloat(Minimum=-0.2, Maximum=1)
    Index 0 Param 2=1
    Index 0 Param 3=&RandomRangeFloat(Minimum=0, Maximum=0.5)
    Index 0 Param 4=1
    Index 1 Param 1=&RandomRangeFloat(Minimum=0.05, Maximum=0.1)
    Index 1 Param 2=&FloatFromCurve(FloatCurve=DataInterface() {propertyValues {Curve {keys [<at 0 0.9>,<at 1 0.4>]}}}, Scale Curve=0.25)
    Index 1 Param 3=&FloatFromCurve(
     FloatCurve=DataInterface() {propertyValues {Curve {keys [
         <at 0 0>
         {interpMode=RCIM_Cubic, tangentMode=RCTM_User, time=0.05, value=-0.05}
         {interpMode=RCIM_Cubic, tangentMode=RCTM_User, time=0.1, value=1.6}
         {interpMode=RCIM_Cubic, tangentMode=RCTM_User, time=0.2, value=0.1}
         {interpMode=RCIM_Cubic, tangentMode=RCTM_User, time=0.5, value=0.75}
         {interpMode=RCIM_Cubic, tangentMode=RCTM_User, time=0.6, value=0.7}
         {interpMode=RCIM_Cubic, tangentMode=RCTM_User, time=1, value=1.75}
        ]}}}
     Scale Curve=&RandomRangeFloat(Minimum=0.1, Maximum=0.3)
    )
    Index 3 Param 3=&RandomRangeFloat(Minimum=800, Maximum=1500)
    Module.Param0WriteEnabled=true
    Module.Param0WriteEnabled1=true
    Module.Param0WriteEnabled3=true
    Module.Param1WriteEnabled=true
    Module.Param1WriteEnabled1=true
    Module.Param1WriteEnabled3=true
    Module.Param2WriteEnabled=true
    Module.Param2WriteEnabled1=true
    Module.Param2WriteEnabled3=true
    Module.Param3WriteEnabled=true
    Module.Param3WriteEnabled1=true
    Module.Param3WriteEnabled3=true
    Write Parameter Index 1=true
    Write Parameter Index 3=true
   )
   &MaintainInCameraParticleScale() {enabled=false}
   &ScaleSpriteSize(
    Uniform Curve Scale=0.25
    Uniform Curve Sprite Scale=DataInterface() {propertyValues {Curve {keys [
        {interpMode=RCIM_Cubic, tangentMode=RCTM_User, time=0, value=0}
        {interpMode=RCIM_Cubic, tangentMode=RCTM_User, time=0.1, value=1}
        <at 0.35 0.625>
        {interpMode=RCIM_Cubic, tangentMode=RCTM_User, time=0.6, value=0.625}
        {interpMode=RCIM_Cubic, tangentMode=RCTM_User, time=0.8, value=0.875}
        {tangentMode=RCTM_User, time=1, value=0}
       ]}}}
   ) {enabled=false}
   &ScaleSpriteSize() {enabled=false, module=ScaleSpriteSize001}
   &ScaleSpriteSize(
    Scale Factor=&MakeVector2D(
     X=&FloatFromCurve(
      FloatCurve=DataInterface() {propertyValues {Curve {keys [
          {interpMode=RCIM_Cubic, tangentMode=RCTM_User, time=0, value=0.2}
          {interpMode=RCIM_Cubic, tangentMode=RCTM_User, time=0.3, value=0.5}
          {interpMode=RCIM_Cubic, tangentMode=RCTM_User, time=1, value=0.375}
         ]}}}
      Scale Curve=0.35
     )
     Y=&FloatFromCurve(
      FloatCurve=DataInterface() {propertyValues {Curve {keys [
          {interpMode=RCIM_Cubic, tangentMode=RCTM_User, time=0, value=0}
          {arriveTangent=1.6606, interpMode=RCIM_Cubic, leaveTangent=1.6606, tangentMode=RCTM_User, time=0.2, value=0.8}
          {interpMode=RCIM_Cubic, tangentMode=RCTM_User, time=1, value=0.65}
         ]}}}
      Scale Curve=0.25
     )
    )
    Scale Sprite Size Mode=::NewEnumerator2
   ) {module=ScaleSpriteSize002}
   /SetVariables(Particles.New Vector 2f:Vector2f=&Multiply_Vector2D_ByFloat(
     Float=&FloatFromCurve(FloatCurve=DataInterface() {propertyValues {Curve {keys [<at 0 0>,<at 0.2 1>,<at 0.5 0>]}}}, Scale Curve=10)
     Vector2D=<1 1>
    )) {isSetParameters=true, module=SetVariables_#8}
   /SetVariables(Particles.New Linear Color:LinearColor=&MultiplyLinearColor(
     Linear Color=<rgba 0.0729 0.0185 0.0629 5>
     Scale Factor=&MakeVector4_FromFloat(Float=&FloatFromCurve(FloatCurve=DataInterface() {propertyValues {Curve {keys [<at 0 1>,<at 0.4 0>]}}}))
    )) {isSetParameters=true, module=SetVariables_#9}
   &SpriteFacingAndAlignment(Sprite Facing=<0 0 1>, WriteAlignment=false, WriteFacing=true)
  ]
  rendererClasses [&NiagaraSpriteRendererProperties]
  renderers [
   {rendererClass=&NiagaraSpriteRendererProperties}
   {rendererClass=&NiagaraSpriteRendererProperties, rendererIndex=1}
   {rendererClass=&NiagaraSpriteRendererProperties, rendererIndex=2}
  ]
 }
 Chrome_Blobs003 {
  simTarget=GPUComputeSim
  emitterUpdateScript [
   /SetVariables(Emitter.CameraPostion:Vector3f=&ReturnCameraVector(Camera Property=::NewEnumerator0, Coordinate Space=::Simulation)) {isSetParameters=true, module=SetVariables_#1}
   &EmitterState()
   &SpawnParticlesFromOtherEmitter(
    Attribute Reader=DataInterface() {propertyValues {EmitterBinding {bindingMode=Other, emitterName=ChromeBigShapes}}}
    Maximum Age Spawn Threshold=1.25
    Minimum Age Spawn Threshold=1
    Spawn Rate=3
    Spawn Rate Per Particle Cap=3
    Use Maximum Age=false
    Use Minimum Age=false
   )
  ]
  particleSpawnScript [
   &InitializeParticle(
    AdjustAlpha=true
    AdjustHue=true
    AdjustSaturation=true
    AdjustValue=true
    Alpha Scale Range:Vector2f=<0.8 0.95>
    Color=<rgba 0.0818 0.1045 0.9792 1>
    Color Mode=::NewEnumerator3
    Lifetime Max=5
    Lifetime Min=4
    Lifetime Mode=::NewEnumerator1
    Saturation Range:Vector2f=<0.25 0.5>
    Sprite Rotation Angle Max=0
    Sprite Rotation Mode=::NewEnumerator2
    Sprite Size Max=<40 50>
    Sprite Size Min=<30 40>
    Sprite Size Mode=::NewEnumerator2
    UsePositionOffset=false
    Value Range:Vector2f=<0.25 0.5>
    Write Material Random=true
   ) {module=InitializeParticle001}
   &SampleParticlesFromOtherEmitter(
    Age Sampling:ENiagara_DirectReadSamplingMode=::NewEnumerator0
    Applied Position Offset:Vector3f=&RandomRangeVector(Minimum=<0 0 0>, Maximum=<1 1 1>)
    Color Sampling=::NewEnumerator0
    Maximum Sampled Particle Age=0.75
    Minimum Sampled Particle Age=0.5
    Module.MaxSampledAgeCutoff=true
    Module.MinSampledAgeCutoff=true
    Reject Particle By Sampled Age=true
   )
   &SubUVAnimation(End Frame=3, SubUV Animation Mode:ENiagara_SubUVLookupMode=::NewEnumerator2)
   /SetVariables(Particles.SizeStretch:Vector2f=&Multiply_Vector2D(
     A=$Particles.SpriteSize
     B=&MakeVector2D(
      X=&RandomRangeFloat(Minimum=5, Maximum=10)
      Y=&RandomRangeFloat(Minimum=25, Maximum=50)
     )
    )) {isSetParameters=true, module=SetVariables_#7}
   &CameraOffset(Camera Offset Amount=-10)
   &self:ScratchModule(Angle=35, Normal:Vector3f=<0 0 1>, Velocity=1) {module=ScratchModule}
   &ShapeLocation(MaskSpawnGroup=false, Non Uniform Scale=<1 1 0.1>, Override Seed=false, Sphere Radius=5, UseOwnerScale=true)
   /Niagara/Modules/Update/Forces/V2/CurlNoiseForce(Noise Frequency=250, Noise Strength=2500, Pan Noise=false) {enabled=false}
   &ApplyInitialForces()
  ]
  particleUpdateScript [
   &ParticleState()
   &self:Camera_to_Particle_Distance_Calculation(Camera Postion:Vector3f=$Emitter.CameraPostion, Distance=5000, Offset=250, Particle Position:NiagaraPosition=$Particles.Position) {module=Camera_to_Particle_Distance_Calculation}
   &GravityForce(Gravity=&Multiply_VectorByFloat(
     Vector=&RandomRangeVector(Minimum=<-3 -3 15>, Maximum=<3 3 25>)
     Float=&FloatFromCurve(
      FloatCurve=DataInterface() {propertyValues {Curve {keys [<at 0 0>,<at 0.25 0>,<at 1 1>]}}}
      Scale Curve=&MakeFloatFromVector(Vector=$Particles.Initial.Velocity, Channel=::NewEnumerator2)
     )
    ))
   &SolveForcesAndVelocity(Clamp Velocity=true, Limit Acceleration=true, Speed Limit=50)
   &ScaleColor(
    Scale Alpha=&FloatFromCurve(FloatCurve=DataInterface() {propertyValues {Curve {keys [
         {interpMode=RCIM_Cubic, tangentMode=RCTM_User, time=0, value=0}
         {interpMode=RCIM_Cubic, tangentMode=RCTM_User, time=0.1, value=1}
         <at 0.7 1>
         {tangentMode=RCTM_User, time=1, value=0}
        ]}}})
    Scale RGB=&MakeVectorFromLinearColorRGB(Color=$Particles.Initial.Color)
    ScaleA=true
    ScaleRGB=true
   )
   &DynamicMaterialParameters(
    Index 0 Param 1=&RandomRangeFloat(Minimum=-0.2, Maximum=1)
    Index 0 Param 2=1
    Index 0 Param 3=&RandomRangeFloat(Minimum=0, Maximum=0.5)
    Index 0 Param 4=1
    Index 1 Param 1=&RandomRangeFloat(Minimum=0.05, Maximum=0.1)
    Index 1 Param 2=&FloatFromCurve(FloatCurve=DataInterface() {propertyValues {Curve {keys [<at 0 0.9>,<at 1 0.4>]}}}, Scale Curve=0.25)
    Index 1 Param 3=&FloatFromCurve(
     FloatCurve=DataInterface() {propertyValues {Curve {keys [
         <at 0 0>
         {interpMode=RCIM_Cubic, tangentMode=RCTM_User, time=0.05, value=-0.05}
         {interpMode=RCIM_Cubic, tangentMode=RCTM_User, time=0.1, value=1.6}
         {interpMode=RCIM_Cubic, tangentMode=RCTM_User, time=0.2, value=0.1}
         {interpMode=RCIM_Cubic, tangentMode=RCTM_User, time=0.5, value=0.75}
         {interpMode=RCIM_Cubic, tangentMode=RCTM_User, time=0.6, value=0.7}
         {interpMode=RCIM_Cubic, tangentMode=RCTM_User, time=1, value=1.75}
        ]}}}
     Scale Curve=&RandomRangeFloat(Minimum=0.1, Maximum=0.3)
    )
    Index 3 Param 3=&RandomRangeFloat(Minimum=800, Maximum=1500)
    Module.Param0WriteEnabled=true
    Module.Param0WriteEnabled1=true
    Module.Param0WriteEnabled3=true
    Module.Param1WriteEnabled=true
    Module.Param1WriteEnabled1=true
    Module.Param1WriteEnabled3=true
    Module.Param2WriteEnabled=true
    Module.Param2WriteEnabled1=true
    Module.Param2WriteEnabled3=true
    Module.Param3WriteEnabled=true
    Module.Param3WriteEnabled1=true
    Module.Param3WriteEnabled3=true
    Write Parameter Index 1=true
    Write Parameter Index 3=true
   )
   &MaintainInCameraParticleScale() {enabled=false}
   &ScaleSpriteSize(
    Uniform Curve Scale=0.25
    Uniform Curve Sprite Scale=DataInterface() {propertyValues {Curve {keys [
        {interpMode=RCIM_Cubic, tangentMode=RCTM_User, time=0, value=0}
        {interpMode=RCIM_Cubic, tangentMode=RCTM_User, time=0.1, value=1}
        <at 0.35 0.625>
        {interpMode=RCIM_Cubic, tangentMode=RCTM_User, time=0.6, value=0.625}
        {interpMode=RCIM_Cubic, tangentMode=RCTM_User, time=0.8, value=0.875}
        {tangentMode=RCTM_User, time=1, value=0}
       ]}}}
   ) {enabled=false}
   &ScaleSpriteSize(Uniform Curve Scale=0.25) {module=ScaleSpriteSize001}
   &ScaleSpriteSize(
    Scale Factor=&MakeVector2D(
     X=&FloatFromCurve(
      FloatCurve=DataInterface() {propertyValues {Curve {keys [
          {interpMode=RCIM_Cubic, tangentMode=RCTM_User, time=0, value=0.2}
          {interpMode=RCIM_Cubic, tangentMode=RCTM_User, time=0.3, value=0.5}
          {interpMode=RCIM_Cubic, tangentMode=RCTM_User, time=1, value=0.375}
         ]}}}
      Scale Curve=0.35
     )
     Y=&FloatFromCurve(
      FloatCurve=DataInterface() {propertyValues {Curve {keys [
          {interpMode=RCIM_Cubic, tangentMode=RCTM_User, time=0, value=0}
          {arriveTangent=1.6606, interpMode=RCIM_Cubic, leaveTangent=1.6606, tangentMode=RCTM_User, time=0.2, value=0.8}
          {interpMode=RCIM_Cubic, tangentMode=RCTM_User, time=1, value=0.65}
         ]}}}
      Scale Curve=0.25
     )
    )
    Scale Sprite Size Mode=::NewEnumerator2
   ) {enabled=false, module=ScaleSpriteSize002}
   /SetVariables(Particles.New Vector 2f:Vector2f=&Multiply_Vector2D_ByFloat(
     Float=&FloatFromCurve(FloatCurve=DataInterface() {propertyValues {Curve {keys [<at 0 0>,<at 0.2 1>,<at 0.5 0>]}}}, Scale Curve=10)
     Vector2D=<1 1>
    )) {isSetParameters=true, module=SetVariables_#8}
   /SetVariables(Particles.New Linear Color:LinearColor=&MultiplyLinearColor(
     Linear Color=<rgba 0.0729 0.0185 0.0629 5>
     Scale Factor=&MakeVector4_FromFloat(Float=&FloatFromCurve(FloatCurve=DataInterface() {propertyValues {Curve {keys [<at 0 1>,<at 0.4 0>]}}}))
    )) {isSetParameters=true, module=SetVariables_#9}
   &SpriteFacingAndAlignment(Sprite Facing=<0 0 1>, WriteAlignment=false, WriteFacing=true)
  ]
  rendererClasses [&NiagaraSpriteRendererProperties]
  renderers [{rendererClass=&NiagaraSpriteRendererProperties}]
 }
 No {
  bEnabled=false
  simTarget=GPUComputeSim
  emitterUpdateScript [
   /SetVariables(Emitter.CameraPostion:Vector3f=&ReturnCameraVector(Camera Property=::NewEnumerator0, Coordinate Space=::Simulation)) {isSetParameters=true, module=SetVariables_#1}
   &EmitterState()
   &SpawnRate(SpawnRate=&Multiply_Float(A=1000, B=1), Use Spawn Probability=true)
  ]
  particleSpawnScript [
   /Niagara/Modules/Spawn/Initialization/InitializeParticle(
    Color=&LerpLinearColors(StartColor=<rgba 1 1 1 1>, EndColor=<rgba 1 1 1 1>, LerpFactor=1)
    Lifetime=&Multiply_Float(A=&UniformRangedFloat(Minimum=1, Maximum=2), B=1)
    Module.Write Color=true
    Module.Write Lifetime=true
    Module.Write Mass=true
    Module.Write Material Random=true
    Module.Write Position=false
    Module.Write Scale=false
    Module.Write SpriteRotation=false
    Module.Write SpriteSize=true
    Position=/Niagara/Functions/Localspace/SimulationPosition()
    Sprite Size=&UniformRangedVector2D(
     Minimum=&Vector2DFromFloat(Value=25)
     Maximum=&Vector2DFromFloat(Value=50)
    )
   ) {enabled=false}
   /SetVariables(Particles.VisibilityTag=1) {enabled=false, isSetParameters=true, module=SetVariables_#2}
   /Niagara/Modules/Spawn/Initialization/V2/InitializeParticle(
    AdjustAlpha=false
    AdjustHue=true
    AdjustSaturation=false
    AdjustValue=false
    Color=<rgba 0.9258 0.7969 1 1>
    Color Mode=::NewEnumerator3
    Hue Shift Range:Vector2f=<-0.25 0.25>
    Lifetime Max=4
    Lifetime Min=3
    Lifetime Mode=::NewEnumerator1
    Sprite Rotation Angle Max=5
    Sprite Rotation Angle Min=-5
    Sprite Rotation Mode=::NewEnumerator2
    Sprite Size Max=<125 250>
    Sprite Size Min=<100 150>
    Sprite Size Mode=::NewEnumerator2
    UsePositionOffset=false
    Write Material Random=true
   ) {module=InitializeParticle001}
   &BoxLocation(
    Box Offset=&MakeVector(
     X=15
     Y=15
     Z=&Multiply_Float(A=&MakeFloatFromVector(Vector=$System.BoxSize, Channel=::NewEnumerator2), B=0)
    )
    Box Origin:NiagaraPosition=/Niagara/DynamicInputs/Helpers/SimulationPosition()
    Box Size=$System.BoxSize
    Coordinate Space:ENiagaraCoordinateSpace=::Simulation
    MaskSpawnGroup=false
    Module.Override Seed=false
    Rotate Box=false
    Surface Only=false
   ) {enabled=false}
   &FN_Spawn_GBuffer_Color() {enabled=false}
   /Niagara/Modules/Update/Forces/CurlNoiseForce(Noise Frequency=150, Pan Noise=true, Pan Noise Field=<0 0 -5>) {enabled=false, module=CurlNoiseForce001}
   &ApplyInitialForces() {enabled=false}
   &CameraOffset(Camera Offset Amount=25)
   &SphereLocation(
    MaskSpawnGroup=false
    Override Seed=false
    Sphere Radius=1500
    Surface Only=false
    UseNonUniformScale=false
    UseOffset=false
    UseSphereOrientation=false
   )
   &MoveToNearestDistanceFieldSurface_GPU(Kill by Distance=50, bKillByDistance=true)
   /SetVariables(Particles.InitialPos:NiagaraPosition=$Particles.Position) {isSetParameters=true, module=SetVariables_#3}
   &PlaceParticlesOnDepthBuffer_GPU(Max Z Depth=3500, Module.Use Max Z Depth=true)
   &KillParticles(Kill Particles=&SetBoolByFloatComparison(
     A=&VectorLength(Vector=&Subtract_Vector(A=$Particles.InitialPos, B=Linked() {linkedVariable {name=Particles.Position, type=NiagaraPosition}}))
     B=3500
    ))
   &KillParticles(Kill Particles=&Boolean_NotOperation(Bool=$Output.PlaceParticlesOnDepthBuffer_GPU.ValidSample)) {module=KillParticles001}
   &FN_Spawn_GBuffer_Color_WithinRange(ColorMaxRange=1, ColorMinRange=0.04) {enabled=false}
  ]
  particleUpdateScript [
   &ParticleState()
   &KillParticlesInVolume(
    Box Size:Vector3f=$System.BoxSize
    Invert Volume=true
    Kill Shape=::NewEnumerator1
    Origin Offset:Vector3f=&MakeVector(
     X=0
     Y=0
     Z=&Multiply_Float(A=&MakeFloatFromVector(Vector=$System.BoxSize, Channel=::NewEnumerator2), B=0)
    )
   ) {enabled=false}
   &self:Camera_to_Particle_Distance_Calculation(Camera Postion:Vector3f=$Emitter.CameraPostion, Distance=5000, Offset=250, Particle Position:NiagaraPosition=$Particles.Position) {module=Camera_to_Particle_Distance_Calculation}
   /SetVariables() {enabled=false, isSetParameters=true, module=SetVariables_#4}
   /SetVariables(Particles.GradientColor:Vector3f=&MakeVectorFromLinearColorRGB(Color=&MakeLinearColorFromVectorAndFloat(
      "Vector (RGB)"=&MakeVectorFromLinearColorRGB(Color=&ColorFromCurve(ColorCurve=DataInterface() {propertyValues {
          AlphaCurve {keys [<at -0.0023 1 RCIM_Cubic>,<at 0.9977 1 RCIM_Cubic>]}
          BlueCurve {keys [<at 0.1175 0.3005 RCIM_Cubic>,<at 0.2682 0.2705>,<at 0.4507 0.2122>,<at 0.5951 0.5271>,<at 0.7691 1.5>,<at 0.9071 0.521 RCIM_Cubic>]}
          GreenCurve {keys [<at 0.1175 0.1878 RCIM_Cubic>,<at 0.2682 0.2623>,<at 0.4507 0.2346>,<at 0.5951 0.4287>,<at 0.7691 0.4532>,<at 0.9071 0.4851 RCIM_Cubic>]}
          RedCurve {keys [<at 0.1175 0.3372 RCIM_Cubic>,<at 0.2682 0.7305>,<at 0.4507 0.7758>,<at 0.5951 0.7084>,<at 0.7691 0>,<at 0.9071 0.807 RCIM_Cubic>]}
         }}))
      "Float (Alpha)"=1
     ))) {enabled=false, isSetParameters=true, module=SetVariables_#5}
   /SetVariables(Particles.VisibilityTag=&MakeCustomIntFromBool(Bool=$Transient.FirstFrame, True INT=1, False INT=0)) {enabled=false, isSetParameters=true, module=SetVariables_#6}
   &GravityForce(Gravity=&Multiply_VectorByFloat(
     Vector=<0 0 15>
     Float=&FloatFromCurve(FloatCurve=DataInterface() {propertyValues {Curve {keys [<at 0 0>,<at 0.25 0>,<at 1 1>]}}})
    ))
   &CurlNoiseForce(
    Curl Noise Cone Mask Angle=150
    Curl Noise Cone Mask Falloff Angle=25
    Mask Curl Noise=true
    Noise Frequency:NiagaraFloat=&RandomRangeFloat(Minimum=250, Maximum=500)
    Noise Strength=&FloatFromCurve(
     FloatCurve=DataInterface() {propertyValues {Curve {keys [
         {interpMode=RCIM_Cubic, tangentMode=RCTM_User, time=0, value=0}
         {interpMode=RCIM_Cubic, tangentMode=RCTM_User, time=0.0675, value=1.5837}
         {interpMode=RCIM_Cubic, tangentMode=RCTM_User, time=0.1302, value=0.1555}
         {interpMode=RCIM_Cubic, tangentMode=RCTM_User, time=0.25, value=1}
        ]}}}
     Scale Curve=100
    )
    Pan Noise=true
    Pan Noise Field=<0 1 80>
    Random Seed=25
   ) {enabled=false}
   &PointAttractionForce(
    Attraction Radius=400
    Attraction Strength=-1
    Attractor Position:Vector3f=$Emitter.CameraPosition
    Attractor Position Offset:Vector3f=<0 0 -500>
    Falloff Exponent=1
    Kill Within Radius=false
    Module.Use Falloff=true
    UseAttractorOffset=true
   ) {enabled=false, module=PointAttractionForce001}
   &Drag(
    Drag=&Multiply_Float(
     A=&UniformRangedFloat(Minimum=0, Maximum=0)
     B=&FloatFromCurve(FloatCurve=DataInterface() {propertyValues {Curve {keys [<at 0.1 1>,<at 0.25 0.1>]}}})
    )
    Use Linear Drag=true
    Use Rotational Drag=false
   ) {enabled=false}
   &SolveForcesAndVelocity(Clamp Velocity=true, Limit Acceleration=false, Speed Limit=250)
   &ScaleColor(
    Scale Alpha=&FloatFromCurve(FloatCurve=DataInterface() {propertyValues {Curve {keys [
         {interpMode=RCIM_Cubic, tangentMode=RCTM_User, time=0, value=0}
         {interpMode=RCIM_Cubic, tangentMode=RCTM_User, time=0.1, value=1}
         <at 0.7 1>
         {tangentMode=RCTM_User, time=1, value=0}
        ]}}})
    Scale RGB=<5 5 5>
    ScaleA=true
    ScaleRGB=true
   )
   &DynamicMaterialParameters(
    Index 0 Param 1=&RandomRangeFloat(Minimum=-0.2, Maximum=1)
    Index 0 Param 2=1
    Index 0 Param 3=&RandomRangeFloat(Minimum=0, Maximum=0.5)
    Index 0 Param 4=1
    Index 1 Param 1=&RandomRangeFloat(Minimum=0.25, Maximum=1)
    Index 1 Param 2=&RandomRangeFloat(Minimum=0.1, Maximum=1)
    Index 1 Param 3=&RandomRangeFloat(Minimum=-0.2, Maximum=1)
    Module.Param0WriteEnabled=true
    Module.Param0WriteEnabled1=true
    Module.Param1WriteEnabled=true
    Module.Param1WriteEnabled1=true
    Module.Param2WriteEnabled=true
    Module.Param2WriteEnabled1=true
    Module.Param3WriteEnabled=true
    Module.Param3WriteEnabled1=true
    Write Parameter Index 1=true
   )
   &MaintainInCameraParticleScale() {enabled=false}
   &ScaleSpriteSize(Uniform Curve Scale=0.1)
  ]
  rendererClasses [&NiagaraSpriteRendererProperties]
  renderers [{rendererClass=&NiagaraSpriteRendererProperties}]
 }
}
summary {
 systemName=NS_LackLuster_Chrome_InWorld
 emitters [
  {bEnabled=false, emitterName=Chrome_Blobs, simTarget=GPUComputeSim, rendererClasses [&NiagaraSpriteRendererProperties]}
  {bEnabled=false, emitterName=No, simTarget=GPUComputeSim, rendererClasses [&NiagaraSpriteRendererProperties]}
  {emitterName=Chrome_Blobs002, simTarget=GPUComputeSim, rendererClasses [&NiagaraSpriteRendererProperties]}
  {emitterName=ChromeBigShapes, simTarget=GPUComputeSim, rendererClasses [&NiagaraSpriteRendererProperties]}
  {bEnabled=false, emitterName=Chrome_Blobs001, simTarget=GPUComputeSim, rendererClasses [&NiagaraSpriteRendererProperties]}
  {emitterName=Chrome_Blobs003, simTarget=GPUComputeSim, rendererClasses [&NiagaraSpriteRendererProperties]}
 ]
 userVariables []
}
